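; Destructible large bush prop for the WildEstate foliage set, with custom death effects and a lighting tweak.
; The construction script stops its static mesh affecting distance field lighting.
; On death it calls the parent effects, then on mobile only spawns a toon-tree destruction Niagara system at the actor's transform, applies FX customization to it and debug-prints the result.
; BeginPlay, overlap and Tick are disabled stubs.
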(event EventBeginPlay
  (|CallParentFunction) ; This node is disabled and will not be called.\nDrag off pins to build functionality.
)

(event Collision|EventActorBeginOverlap
  (params (OtherActor :type Actor%20Object%20Reference))
  (|CallParentFunction OtherActor) ; This node is disabled and will not be called.\nDrag off pins to build functionality.
)

(event EventTick
  (params (DeltaSeconds :type Float%20%28single-precision%29))
  (|CallParentFunction DeltaSeconds) ; This node is disabled and will not be called.\nDrag off pins to build functionality.
)

(event EventOnDeathPlayEffects
  (params (Damage :type Float%20%28single-precision%29) (DamageTags :type Gameplay%20Tag%20Container%20Structure%20%28by%20ref%29) (Momentum :type Vector :default "0, 0, 0") (HitInfo :type Hit%20Result%20Structure%20%28by%20ref%29) (InstigatedBy :type Fort%20Pawn%20Object%20Reference) (DamageCauser :type Actor%20Object%20Reference) (EffectContext :type Gameplay%20Effect%20Context%20Handle%20Structure))
  (bind (_returnvalue_location _returnvalue_rotation) (Math|Transform|BreakTransform (Transformation|GetActorTransform)))
  (|CallParentFunction Damage DamageTags Momentum HitInfo InstigatedBy DamageCauser EffectContext)
  (if (TouchPlatform|IsMobileGame)
    (bind _returnvalue (Niagara|SpawnSystematLocation "/Game/Effects/Fort_Effects/World/Destruction/Destruction2/Trees/NS_Destruction_Toon_Trees.NS_Destruction_Toon_Trees" _returnvalue_location _returnvalue_rotation))
    (bind _returnvalue_2 (Class|FortFXCustomizationComponent|ApplyFXCustomizationtoComponent self _returnvalue))
    (Development|PrintString (Utilities|String|Append "Customization Applied:" (Utilities|String|ToString(Boolean) _returnvalue_2)))))

(fn ConstructionScript
  (|CallParentFunction)
  (Rendering|SetAffectDistanceFieldLighting (Variables|Debug|GetStaticMeshComponent) false))
